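# Material function that outputs a GrassColor for LimeRock grass, tinted by a color parameter and varied across the world.
# It samples the WildEstate landscape grass overlay mask at two scales, using world XY position divided by size parameters as UVs.
# The finer sample's red channel multiplies the tint, while the broader sample's green channel lerps a brightness factor from a minimum darkness up to full, breaking up uniformity.
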MaterialFunction /LimeRock_Tree/Assets/Treehouse/Materials/Grass/MF_LimeRockGrass
graph {nodes [
  n1 = ScalarParameter() {defaultValue=0.76, parameterName=GrassNoise_MinDarkness}
  n2 = TextureSample(UVs=Divide(A=WorldPosition().XY, B=ScalarParameter() {defaultValue=5379.1084, parameterName=NoiseTextureSize})) {samplerType=SAMPLERTYPE_Masks, texture=&/WildEstate_Landscape_Grass_Overlay_D}
  n3 = TextureSample(UVs=Divide(A=WorldPosition().XY, B=ScalarParameter() {defaultValue=512, parameterName=OverlayTextureSize})) {samplerType=SAMPLERTYPE_Masks, texture=&WildEstate_Landscape_Grass_Overlay_D}
  n4 = VectorParameter() {parameterName=Color_Grass, defaultValue=<rgba 0.0176 0.2604 0.0195 1>}
  FunctionOutput(None=Multiply(A=LinearInterpolate(A=@n1, Alpha=@n2.G), B=Multiply(A=@n3.R, B=@n4.RGB))) {outputName=GrassColor}
 ]}
