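# Default Lit material for the LimeRock Treehouse grass floor, tagged with the Grass physical material.
# It builds material attributes from the GrassColor output of MF_LimeRockGrass as base color, plus Specular and Roughness scalar parameters, with roughness fully rough by default.
# Normal, World Position Offset and Opacity Mask slots are declared but unconnected.
# The attributes then pass through the shared MF_LimeRock_UnlitShading function.
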Material /LimeRock_Tree/Assets/Treehouse/Materials/Grass/M_LimeRockGrassFloor
graph {
 properties {
  bAutomaticallySetUsageInEditor=false
  bUseMaterialAttributes=true
  physMaterial=/Game/FortressPhysicalMaterials/Grass
  usedShadingModels=Default Lit
 }
 nodes [
  n1 = SetMaterialAttributes!(
   MaterialAttributes
   Base Color=MaterialFunctionCall() {materialFunction=&/MF_LimeRockGrass}.GrassColor
   Specular=ScalarParameter() {parameterName=Specular}
   Roughness=ScalarParameter() {defaultValue=1, parameterName=Roughness}
   Normal
   World Position Offset
   Opacity Mask
  ) {attributeSetTypes [
    #1
    #2
    #3
    #4
    #5
    #6
   ]}
  MaterialFunctionCall!(In=@n1) {materialFunction=/LimeRock_Tree/Assets/Treehouse/MaterialFunctions/MF_LimeRock_UnlitShading}
 ]
}
