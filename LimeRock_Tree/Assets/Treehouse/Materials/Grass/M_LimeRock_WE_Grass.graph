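# Masked material for LimeRock instanced Nanite grass, tinting a grass color function with a texture-driven outline and passing it through an unlit-shading function.
# Normals follow instance up, flattening shading across blades.
# World position offset combines height-masked sine sway, blended toward storm settings via the wind actor function, with camera-facing rotation and bend.
# Camera distance plus per-instance randomness collapses vertices toward an instance point for staggered culling.
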Material /LimeRock_Tree/Assets/Treehouse/Materials/Grass/M_LimeRock_WE_Grass
graph {
 properties {
  bAutomaticallySetUsageInEditor=false
  bTangentSpaceNormal=false
  bUseMaterialAttributes=true
  bUsedWithInstancedStaticMeshes=true
  bUsedWithNanite=true
  blendMode=BLEND_Masked
  usedShadingModels=Default Lit
 }
 nodes [
  n1 = PerInstanceRandom()
  n2 = Add(A=Time(), B=@n1)
  n3 = ScalarParameter() {defaultValue=0.5257, group=Animation, parameterName=Test_Anim_Speed}
  n4 = Multiply(A=@n2, B=@n3)
  n5 = ScalarParameter() {defaultValue=2, group=Animation, parameterName=Test_Anim_Speed_Storm}
  n6 = Multiply(A=@n2, B=@n5)
  n7 = Multiply(A=@n1) {constB=0.01}
  n8 = Add(A=@n7, B=@n3)
  n9 = Add(A=@n7, B=@n5)
  n10 = LinearInterpolate(A=Sine(None=Multiply(A=@n2, B=@n8)), B=Sine(None=Multiply(A=@n2, B=@n9)), Alpha=NamedRerouteUsage())
  n11 = Add(A=LinearInterpolate(A=Sine(None=@n4), B=Sine(None=@n6), Alpha=NamedRerouteUsage()), B=@n10)
  Add(A=@n11, B=ScalarParameter() {parameterName=Anim_Bias})
  Add(A=@n10) {constB=0.5}
  CameraVectorWS()
  Constant3Vector()
  CrossProduct()
  Divide(A=Add(A=@n10))
  n12 = TextureSample() {texture=&/T_WIldEstate_Grass_Standard_M}
  n13 = LinearInterpolate(A=ScalarParameter() {defaultValue=-0.552, parameterName=OutlineIntensity}, Alpha=@n12.G)
  n14 = Constant3Vector() {constant=<rgba 0 0 1 0>}
  n15 = SetMaterialAttributes!(
   MaterialAttributes
   Base Color=Multiply(
    A=Saturate(None=@n13)
    B=MaterialFunctionCall() {materialFunction=&/MF_LimeRockGrass}.GrassColor
   )
   Specular=ScalarParameter() {parameterName=Specular}
   Roughness=ScalarParameter() {defaultValue=1, parameterName=Roughness}
   Normal=Normalize(VectorInput=Transform(None=@n14) {transformSourceType=TRANSFORMSOURCE_Instance})
   World Position Offset=NamedRerouteUsage()
   Opacity Mask=@n12.R
  ) {attributeSetTypes [
    #1
    #2
    #3
    #4
    #5
    #6
   ]}
  MaterialFunctionCall!(In=@n15) {materialFunction=/LimeRock_Tree/Assets/Treehouse/MaterialFunctions/MF_LimeRock_UnlitShading}
  n16 = MaterialFunctionCall!(
   Position=NamedRerouteUsage()
   IsTrees=StaticBool() {value=true}
   IsGrass
   IsMediumPlant
   IsMediumPlant_2024Anim=StaticBool() {value=true}
   IsMediumPlant2ndLevel
   MediumPlant_BasePivotPostion
  ) {materialFunction=/Game/Environments/Asteria/Foliage/Wind/MF_WindActorData}
  NamedRerouteDeclaration(None=@n16.WindIntensitySingleSample) {name=WindIntensity, nodeColor=<rgba 0 1 0.3438 1>}
  n17 = LocalPosition()
  n18 = ScalarParameter() {defaultValue=6.5821, parameterName=BendGradient}
  n19 = WorldPosition()
  n20 = NamedRerouteDeclaration(None=TransformPosition(None=Constant3Vector()) {transformSourceType=TRANSFORMPOSSOURCE_Instance, transformType=TRANSFORMPOSSOURCE_World}) {name=InstancePivot, nodeColor=<rgba 0 0.4375 1 1>}
  n21 = Subtract(A=Add(A=NamedRerouteUsage(), B=WorldPosition() {worldPositionShaderOffset=WPT_ExcludeAllShaderOffsets}.XYZ), B=@n20)
  n22 = Normalize(VectorInput=Transform(None=Constant3Vector() {constant=<rgba 1 0 0 0>}) {transformSourceType=TRANSFORMSOURCE_Instance})
  n23 = Normalize(VectorInput=Transform(None=Constant3Vector() {constant=<rgba 0 1 0 0>}) {transformSourceType=TRANSFORMSOURCE_Instance})
  n24 = Normalize(VectorInput=Transform(None=Constant3Vector() {constant=<rgba 0 0 1 0>}) {transformSourceType=TRANSFORMSOURCE_Instance})
  n25 = Subtract(
   A=TransformPosition(None=Constant3Vector()) {transformSourceType=TRANSFORMPOSSOURCE_Instance, transformType=TRANSFORMPOSSOURCE_World}
   B=CameraPositionWS()
  )
  n26 = Multiply(A=Constant3Vector() {constant=<rgba 1 1 0 0>}, B=@n25)
  n27 = Normalize(VectorInput=Transform(None=@n26) {transformSourceType=TRANSFORMSOURCE_Instance})
  n28 = Constant3Vector() {constant=<rgba 0 0 1 0>}
  n29 = Transform(None=@n25) {transformSourceType=TRANSFORMSOURCE_Instance}
  n30 = Normalize(VectorInput=LinearInterpolate(A=@n28, B=Normalize(VectorInput=@n29), Alpha=ScalarParameter() {defaultValue=-90, parameterName=Bendaway}))
  n31 = Transform(None=CrossProduct(A=@n30, B=@n27)) {transformSourceType=TRANSFORMSOURCE_Instance}
  n32 = Subtract(
   A=MaterialFunctionCall(
    VectorToTransform=MaterialFunctionCall(VectorToTransform=@n21, BasisX=@n22, BasisY=@n23, BasisZ=@n24) {materialFunction=&Transform3x3Matrix}.Transformed Vector
    BasisX=@n27
    BasisY=Normalize(VectorInput=@n31)
    BasisZ=@n30
   ) {materialFunction=&Transform3x3Matrix}.Transformed Vector
   B=@n21
  )
  n33 = Multiply(A=Subtract(A=Add(A=@n19.XYZ, B=@n32), B=CameraPositionWS()), B=Constant3Vector() {constant=<rgba 1 1 0 0>})
  n34 = Multiply(A=Multiply(A=Divide(A=@n17.Z, B=@n18), B=Normalize(VectorInput=@n33)), B=ScalarParameter() {defaultValue=-4.0319, parameterName=BendPush})
  n35 = NamedRerouteUsage()
  n36 = WorldPosition()
  n37 = Distance(A=@n35, B=CameraPositionWS())
  n38 = ScalarParameter() {defaultValue=1782.2482, parameterName=CullFadeOffset}
  n39 = ScalarParameter() {defaultValue=39072.4492, parameterName=CullDistance}
  n40 = Saturate(None=Divide(A=Subtract(A=@n37, B=@n38), B=@n39))
  n41 = Add(A=PerInstanceRandom(), B=@n40)
  NamedRerouteDeclaration(None=LinearInterpolate(A=Add(A=@n34, B=@n32), B=Subtract(A=@n35, B=@n36.XYZ), Alpha=Saturate(None=@n41))) {name=WPO, nodeColor=<rgba 0.6406 1 0 1>}
  n42 = Constant3Vector() {constant=<rgba 1 1 0 0>}
  n43 = Multiply(A=@n11, B=@n42)
  n44 = LinearInterpolate(
   A=ScalarParameter() {defaultValue=1, parameterName=Wind_Intensity}
   B=ScalarParameter() {defaultValue=50, parameterName=Wind_Intensity_Storm}
   Alpha=NamedRerouteUsage()
  )
  n45 = LocalPosition()
  n46 = ScalarParameter() {defaultValue=3.5322, parameterName=Gradient_Bias}
  n47 = Power(
   Base=Divide(A=Subtract(A=@n45.Z, B=@n46), B=ScalarParameter() {defaultValue=16.6253, parameterName=Gradient_Dist})
   Exp=ScalarParameter() {defaultValue=1.5, parameterName=Gradient_Falloff}
  )
  NamedRerouteDeclaration(None=Multiply(A=Multiply(A=@n43, B=@n44), B=Saturate(None=@n47))) {name=GrassAnim, nodeColor=<rgba 0.9688 1 0 1>}
  NamedRerouteDeclaration(None=@n16.StormMask) {name=StormMask, nodeColor=<rgba 0.7344 1 0 1>}
  NamedRerouteDeclaration(None=@n16.WindI_Direction_WithNoise) {name=WindIntensity_Direction, nodeColor=<rgba 0 1 0.3438 1>}
  NamedRerouteUsage()
  NamedRerouteUsage()
  Normalize(VectorInput=@n26)
  Normalize(VectorInput=Transform(None=Constant3Vector() {constant=<rgba 1 0 0 0>}) {transformSourceType=TRANSFORMSOURCE_Instance})
  n48 = Constant3Vector() {constant=<rgba 0 0 1 0>}
  Normalize(VectorInput=@n48)
  n49 = Constant3Vector() {constant=<rgba 1 0 0 0>}
  Normalize(VectorInput=@n49)
  n50 = Constant3Vector() {constant=<rgba 0 1 0 0>}
  Normalize(VectorInput=@n50)
  OneMinus(None=@n40)
  n51 = Multiply(A=ObjectPositionWS(), B=Constant3Vector() {constant=<rgba 1 1 0 0>})
  RotateAboutAxis(
   NormalizedRotationAxis=Normalize(VectorInput=CrossProduct(A=@n51, B=@n30))
   RotationAngle=ScalarParameter() {defaultValue=-0.08, parameterName=RotationAngle}
   PivotPoint=@n20
   Position=@n21
  )
  n52 = WorldPosition()
  n53 = ObjectPositionWS()
  n54 = Subtract(A=@n52.XYZ, B=@n53)
  n55 = ComponentMask(None=ObjectPositionWS()) {g=true, r=true}
  n56 = ComponentMask(None=CameraPositionWS()) {g=true, r=true}
  n57 = AppendVector(A=Normalize(VectorInput=Subtract(A=@n55, B=@n56)), B=Constant())
  n58 = Constant3Vector() {constant=<rgba 0 0 1 0>}
  n59 = CrossProduct(A=@n57, B=@n58)
  Subtract(
   A=Add(
    A=MaterialFunctionCall(VectorToTransform=@n54, BasisX=@n57, BasisY=@n59, BasisPosition=@n58) {materialFunction=&Transform3x3Matrix}.Transformed Vector
    B=@n53
   )
   B=@n52.XYZ
  )
  Subtract(A=@n38, B=@n39)
  Transform() {transformSourceType=TRANSFORMSOURCE_World}
  Transform(None=Constant3Vector() {constant=<rgba 0 0 1 0>}) {transformSourceType=TRANSFORMSOURCE_Camera, transformType=TRANSFORM_Instance}
  n60 = Subtract(A=Transform(None=Constant3Vector()) {transformSourceType=TRANSFORMSOURCE_Instance}, B=CameraPositionWS())
  n61 = Constant3Vector() {constant=<rgba 1 1 0 0>}
  Transform(None=Normalize(VectorInput=Multiply(A=@n60, B=@n61))) {transformSourceType=TRANSFORMSOURCE_World, transformType=TRANSFORM_Instance}
  Transform(None=@n42) {transformType=TRANSFORM_Camera}
  Transform(None=@n48) {transformSourceType=TRANSFORMSOURCE_Instance, transformType=TRANSFORM_Local}
  Transform(None=@n49) {transformSourceType=TRANSFORMSOURCE_Instance}
  Transform(None=@n50) {transformSourceType=TRANSFORMSOURCE_Instance}
  Transform(None=@n28) {transformSourceType=TRANSFORMSOURCE_Instance}
  VertexInterpolator()
 ]
}
